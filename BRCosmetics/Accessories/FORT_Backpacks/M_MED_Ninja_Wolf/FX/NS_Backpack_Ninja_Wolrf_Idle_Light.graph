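# Niagara idle effect for the Ninja Wolf backpack cosmetic that puts a small pinkish-red light on the backpack.
# A single emitter burst-spawns one two-second particle, drawn only by a light renderer.
# Each update re-samples a bone or socket location on the User.Skeletal Mesh parameter, keeping the light attached.
# FN_Light_v2 scales the light color by time of day, higher for morning and day than for evening and night.
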NiagaraSystem /BRCosmetics/Accessories/FORT_Backpacks/M_MED_Ninja_Wolf/FX/NS_Backpack_Ninja_Wolrf_Idle_Light
dependencies {usedDataInterfaces [], usedDynamicInputs [], usedModules [&SystemState]}
emitters {Light {
  emitterUpdateScript [
   &EmitterState(Inactive Response=::NewEnumerator1, Life Cycle Mode=::NewEnumerator1, Loop Duration=2, UseLoopDelay=true)
   &SpawnBurst_Instantaneous(Spawn Count=1, Use Spawn Probability=false)
  ]
  particleSpawnScript [
   /Niagara/Modules/Spawn/Initialization/InitializeParticle(
    Color=<rgba 1 0.0425 0.3996 1>
    Lifetime=2
    Module.Write Color=true
    Module.Write Lifetime=true
    Module.Write Mass=false
    Module.Write Material Random=true
    Module.Write Position=false
    Module.Write Scale=false
    Module.Write SpriteRotation=false
    Module.Write SpriteSize=true
    Position=&SimulationPosition()
    Sprite Size=<75 75>
   )
   &DynamicMaterialParameters(
    Index 0 Param 1=&UniformRangedFloat(Minimum=0.7, Maximum=0.9)
    Index 0 Param 2=&UniformRangedFloat(Minimum=0.05, Maximum=0.15)
    Index 0 Param 3=1
    Module.Param0WriteEnabled=true
    Module.Param1WriteEnabled=true
    Module.Param2WriteEnabled=false
    Module.Param3WriteEnabled=false
   ) {module=DynamicMaterialParameters002}
  ]
  particleUpdateScript [
   &ParticleState()
   &SkeletalMeshLocation(
    Bone / Socket Index=&SpawnGroup()
    Bone Sampling Mode=::NewEnumerator4
    Inherit Velocity=false
    Interpolate Mesh Sampling=false
    MaskSpawnGroup=false
    Sampled Position Offset Coordinate Space:ENiagaraMeshTransforms=::NewEnumerator3
    Skeletal Mesh=$User.Skeletal Mesh
   )
   &SolveForcesAndVelocity(Clamp Velocity=false, Limit Acceleration=false)
   &FN_Light_v2(
    Light Exponent=2
    Light Radius=20
    LightEnabled=true
    Scale Color:NiagaraFloat=&FNTimeOfDayScalar(Night Scale=5, Morning Scale=10, Day Scale=10, Evening Scale=5)
   )
  ]
  rendererClasses [&NiagaraLightRendererProperties]
  renderers [{rendererClass=&NiagaraLightRendererProperties, rendererIndex=0}]
 }}
summary {
 systemName=NS_Backpack_Ninja_Wolrf_Idle_Light
 emitters [{emitterName=Light, rendererClasses [&NiagaraLightRendererProperties]}]
 userVariables [User.Skeletal Mesh {
   type=NiagaraDataInterfaceSkeletalMesh
   defaultValue {
    struct=NiagaraExt_VariableValue_DataInterface
    value {dataInterface=&self:NiagaraDataInterfaceSkeletalMesh_0}
   }
  }]
}
