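# Physics asset for the M_MED_Ninja_Wolf backpack, setting up its dangling straps and sword for simulation.
# Capsule bodies form two three-segment strap chains hanging from the pack body, with twist locked on the upper and middle joints.
# The long lower capsules, sword and sword helper, are joined with swing and twist locked, closing the chains into a loop so both straps carry the sword as one piece.
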PhysicsAsset /BRCosmetics/Accessories/FORT_Backpacks/M_MED_Ninja_Wolf/Meshes/M_MED_Ninja_Wolf_Pack_Physics
bodies {
 dyn_lower_strap_1_sword [{
   shapeType=Capsule
   length=13
   radius=2.95
   shapeName=dyn_lower_strap_1_sword_capsule
   center=<0 -0.0004 -2.8935>
   rotation=<rot 90 0.0004 -179.9998>
  }]
 dyn_lower_strap_2_swordhelper [{
   shapeType=Capsule
   length=14.75
   radius=2.95
   shapeName=dyn_lower_strap_2_swordhelper_capsule
   center=<0 -0.0004 -2.8935>
   rotation=<rot 90 -0.0002 179.9995>
  }]
 dyn_middle_strap_1 [{shapeType=Capsule, length=1, radius=1.2, shapeName=dyn_middle_strap_1_capsule, center=<2.0435 0 -0.1847>, rotation=<rot 90 0.02 0.0202>}]
 dyn_middle_strap_2 [{shapeType=Capsule, length=1, radius=1.2, shapeName=dyn_middle_strap_2_capsule, center=<1.6611 0 0>, rotation=<rot 90 -0.0034 179.9964>}]
 dyn_upper_strap_1 [{shapeType=Capsule, length=1, radius=1.2, shapeName=dyn_upper_strap_1_capsule, center=<2.0764 0 0>, rotation=<rot 80 90.0002 90.0001>}]
 dyn_upper_strap_2 [{shapeType=Capsule, length=1, radius=1.2, shapeName=dyn_upper_strap_2_capsule, center=<1.8911 0 0>, rotation=<rot 90 2.5305 2.5304>}]
 pack [{shapeType=Capsule, length=70.5, radius=5, shapeName=pack_capsule, center=<-7.9561 0.0001 -4.07>, rotation=<rot -50 0 0>}]
}
constraints [
 {bone1Name=dyn_upper_strap_1, bone2Name=pack, swing1LimitDegrees=45, swing2LimitDegrees=10, twistLimitDegrees=45, twistMotion=Locked}
 {
  bone1Name=dyn_middle_strap_1
  bone2Name=dyn_upper_strap_1
  swing1LimitDegrees=60
  swing2LimitDegrees=10
  twistLimitDegrees=45
  twistMotion=Locked
 }
 {bone1Name=dyn_lower_strap_1_sword, bone2Name=dyn_middle_strap_1, swing1LimitDegrees=45, swing2LimitDegrees=20, twistLimitDegrees=45}
 {bone1Name=dyn_upper_strap_2, bone2Name=pack, swing1LimitDegrees=45, swing2LimitDegrees=45, twistLimitDegrees=45, twistMotion=Locked}
 {
  bone1Name=dyn_middle_strap_2
  bone2Name=dyn_upper_strap_2
  swing1LimitDegrees=45
  swing2LimitDegrees=10
  twistLimitDegrees=45
  twistMotion=Locked
 }
 {bone1Name=dyn_lower_strap_2_swordhelper, bone2Name=dyn_middle_strap_2, swing1LimitDegrees=45, swing2LimitDegrees=45, twistLimitDegrees=45}
 {
  bone1Name=dyn_lower_strap_2_swordhelper
  bone2Name=dyn_lower_strap_1_sword
  swing1LimitDegrees=45
  swing1Motion=Locked
  swing2LimitDegrees=45
  swing2Motion=Locked
  twistLimitDegrees=45
  twistMotion=Locked
 }
]
